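# Niagara effect for the Twin Dark backpack cosmetic that twinkles pale pink sparkle sprites on the backpack.
# Sides, Bottom and Front emitters follow the skeletal mesh with offsets, mirroring sides via spawn groups and trimming the front plane with a triangle test.
# A GPU emitter samples the mesh and kills particles using an FX texture mask at their UV.
# Sprites pulse over life and enlarge with LOD distance to stay readable.
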NiagaraSystem /BRCosmetics/Accessories/FORT_Backpacks/Backpack_F_MED_Twin_Dark/FX/NS_Twin_Dark_Purple_Backpack
dependencies {
 usedDataInterfaces [&NiagaraDataInterfaceVector2DCurve, &NiagaraDataInterfaceCurve, /Script/Niagara.NiagaraDataInterfaceTexture]
 usedDynamicInputs []
 usedModules [&SystemState]
}
emitters {
 Sparkles_Bottom {
  emitterSpawnScript [&FN_PreserveSkydiveUserVariables() {module=PreserveUserVariables}]
  emitterUpdateScript [
   &EmitterState()
   &SpawnRate(Spawn Probability=0.75, Use Spawn Probability=true)
  ]
  particleSpawnScript [
   &InitializeParticle(
    Color=<rgba 1 0.6042 0.9784 1>
    Lifetime Max=1.2
    Lifetime Min=0.7
    Lifetime Mode=::NewEnumerator1
    Sprite Size Mode=::NewEnumerator1
    Uniform Sprite Size Max=14
    Uniform Sprite Size Min=8
    UsePositionOffset=false
    Write Material Random=true
   )
   &CameraOffset(Camera Offset Amount=6)
   /SetVariables("Particles.OffsetPosition ":Vector3f=&MakeVector(
     X=&Lerp_Float(A=-10, B=10, Alpha=$Particles.MaterialRandom)
     Y=&RandomRangeFloat(Minimum=-2, Maximum=-12)
     Z=-20
    )) {isSetParameters=true, module=SetVariables_#1}
  ]
  particleUpdateScript [
   &ParticleState()
   &SkeletalMeshLocation(
    Bone Sampling Mode=::NewEnumerator4
    MaskSpawnGroup=false
    Sampled Position Offset=Linked() {linkedVariable {name="Particles.OffsetPosition ", type=Vector3f}}
    Skeletal Mesh=$User.SkeletalMesh
   )
   &ScaleSpriteSize(Scale Factor=&Vector2DFromCurve(
     Vector2Curve=DataInterface() {propertyValues {
       XCurve {keys [
         {arriveTangent=25.6549, interpMode=RCIM_Cubic, leaveTangent=25.6549, tangentMode=RCTM_User, time=0, value=0.75}
         {arriveTangent=1, interpMode=RCIM_Cubic, leaveTangent=1, time=0.1, value=0.28}
         {arriveTangent=1.3585, leaveTangent=1.3585, time=0.25, value=1}
         {arriveTangent=-1.3333, interpMode=RCIM_Cubic, leaveTangent=-1.3333, time=0.63, value=1}
         {arriveTangent=-5.3669, interpMode=RCIM_Cubic, leaveTangent=-5.3669, tangentMode=RCTM_User, time=1, value=0}
        ]}
       YCurve {keys [
         <at 0 2 RCIM_Cubic>
         {arriveTangent=-4, interpMode=RCIM_Cubic, leaveTangent=-4, time=0.1, value=1}
         <at 0.25 1>
         {arriveTangent=-1.3333, interpMode=RCIM_Cubic, leaveTangent=-1.3333, time=0.63, value=1}
         {arriveTangent=-5.3669, interpMode=RCIM_Cubic, leaveTangent=-5.3669, tangentMode=RCTM_User, time=1, value=0}
        ]}
      }}
     ScaleCurve=<1 1>
    ))
   &ScaleSpriteSize(
    Scale Sprite Size Mode=::NewEnumerator0
    Uniform Scale Factor=&FloatFromCurve(
     FloatCurve=DataInterface() {propertyValues {Curve {keys [<at 500 1 RCIM_Cubic>, {interpMode=RCIM_Cubic, leaveTangent=0.0207, time=1400, value=3}]}}}
     CurveIndex=$Engine.Owner.LODDistance
    )
   ) {module=ScaleSpriteSize001}
  ]
  rendererClasses [&NiagaraSpriteRendererProperties]
  renderers [{rendererClass=&NiagaraSpriteRendererProperties}]
 }
 Sparkles_Front {
  emitterSpawnScript [&FN_PreserveSkydiveUserVariables() {module=PreserveUserVariables}]
  emitterUpdateScript [
   &EmitterState()
   &SpawnRate(SpawnRate=2, Use Spawn Probability=false)
  ]
  particleSpawnScript [
   &InitializeParticle(
    Color=<rgba 1 0.6042 0.9784 1>
    Lifetime Max=1.2
    Lifetime Min=0.8
    Lifetime Mode=::NewEnumerator1
    Sprite Size Mode=::NewEnumerator1
    Uniform Sprite Size Max=12
    Uniform Sprite Size Min=8
    UsePositionOffset=false
    Write Material Random=true
   )
   &CameraOffset(Camera Offset Amount=8)
   &ShapeLocation(
    Box / Plane Mode:ENiagara_BoxPlaneMode=::NewEnumerator1
    MaskSpawnGroup=false
    Offset=<1.5 -13 -12>
    Offset Mode=::NewEnumerator0
    Override Seed=false
    Plane Size:Vector2f=<20 40>
    Rotation Angle=0.135
    Rotation Axis=<0 1 -0.2>
    Rotation Mode=::NewEnumerator1
    Shape Primitive=::NewEnumerator2
    UseOwnerScale=true
   )
  ]
  particleUpdateScript [
   &ParticleState()
   &SkeletalMeshLocation(
    Bone Sampling Mode=::NewEnumerator4
    MaskSpawnGroup=false
    Sampled Position Offset=$Particles.ShapeLocation.ShapePosition
    Skeletal Mesh=$User.SkeletalMesh
   )
   /Niagara/Modules/Math/FindClosestPointOnTriangle(
    Query Position:NiagaraPosition=$Particles.Position
    Triangle Vertex 1:Vector3f=<10 -10 0>
    Triangle Vertex 2:Vector3f=<10 -13 -20>
    Triangle Vertex 3:Vector3f=<-11 -13 -18>
   )
   &ScaleSpriteSize(Scale Factor=&Vector2DFromCurve(
     Vector2Curve=DataInterface() {propertyValues {
       XCurve {keys [
         {arriveTangent=25.6549, interpMode=RCIM_Cubic, leaveTangent=25.6549, tangentMode=RCTM_User, time=0, value=0.75}
         {arriveTangent=1, interpMode=RCIM_Cubic, leaveTangent=1, time=0.1, value=0.28}
         {arriveTangent=1.3585, leaveTangent=1.3585, time=0.25, value=1}
         {arriveTangent=-1.3333, interpMode=RCIM_Cubic, leaveTangent=-1.3333, time=0.63, value=1}
         {arriveTangent=-5.3669, interpMode=RCIM_Cubic, leaveTangent=-5.3669, tangentMode=RCTM_User, time=1, value=0}
        ]}
       YCurve {keys [
         <at 0 2 RCIM_Cubic>
         {arriveTangent=-4, interpMode=RCIM_Cubic, leaveTangent=-4, time=0.1, value=1}
         <at 0.25 1>
         {arriveTangent=-1.3333, interpMode=RCIM_Cubic, leaveTangent=-1.3333, time=0.63, value=1}
         {arriveTangent=-5.3669, interpMode=RCIM_Cubic, leaveTangent=-5.3669, tangentMode=RCTM_User, time=1, value=0}
        ]}
      }}
     ScaleCurve=<1 1>
    ))
   &ScaleSpriteSize(
    Scale Sprite Size Mode=::NewEnumerator0
    Uniform Scale Factor=&FloatFromCurve(
     FloatCurve=DataInterface() {propertyValues {Curve {keys [<at 500 1 RCIM_Cubic>, {interpMode=RCIM_Cubic, leaveTangent=0.0207, time=1400, value=3}]}}}
     CurveIndex=$Engine.Owner.LODDistance
    )
   ) {module=ScaleSpriteSize001}
   &KillParticles(Kill Particles=&InvertBool(Boolean:NiagaraBool=$Output.FindClosestPointOnTriangle.ClosestPointIsOnTriangleSurface))
  ]
  rendererClasses [&NiagaraSpriteRendererProperties]
  renderers [{rendererClass=&NiagaraSpriteRendererProperties}]
 }
 Sparkles_GPU {
  simTarget=GPUComputeSim
  emitterSpawnScript [&FN_PreserveSkydiveUserVariables() {module=PreserveUserVariables}]
  emitterUpdateScript [
   &EmitterState()
   &SpawnRate(SpawnRate=6, Use Spawn Probability=false)
  ]
  particleSpawnScript [&InitializeParticle(
    Color=<rgba 1 0.6042 0.9784 1>
    Lifetime Max=1.2
    Lifetime Min=0.8
    Lifetime Mode=::NewEnumerator1
    Sprite Size Mode=::NewEnumerator1
    Uniform Sprite Size Max=14
    Uniform Sprite Size Min=8
    UsePositionOffset=false
    Write Material Random=true
   )]
  particleUpdateScript [
   &CameraOffset(Camera Offset Amount=6)
   &ParticleState()
   &SkeletalMeshLocation(
    MaskSpawnGroup=false
    Mesh Sampling Type=::NewEnumerator7
    OffsetAlongNormal=false
    OverrideRandomSeed=false
    Skeletal Mesh=$User.SkeletalMesh
   )
   /Niagara/Modules/Update/Textures/SampleTexture(
    Texture=DataInterface() {propertyValues {
      Texture=/BRCosmetics/Accessories/FORT_Backpacks/Backpack_F_MED_Twin_Dark/Textures/T_F_MED_TwinDark_Backpack_FX
      TextureUserParameter {parameter {name=None, typeDefHandle {registeredTypeIndex=2}, varData []}}
     }}
    UV:Vector2f=$Particles.SkeletalMeshLocation.SampledUV
   )
   &ScaleSpriteSize(Scale Factor=&Vector2DFromCurve(
     Vector2Curve=DataInterface() {propertyValues {
       XCurve {keys [
         {arriveTangent=25.6549, interpMode=RCIM_Cubic, leaveTangent=25.6549, tangentMode=RCTM_User, time=0, value=0.75}
         {arriveTangent=1, interpMode=RCIM_Cubic, leaveTangent=1, time=0.1, value=0.28}
         {arriveTangent=1.3585, leaveTangent=1.3585, time=0.25, value=1}
         {arriveTangent=-1.3333, interpMode=RCIM_Cubic, leaveTangent=-1.3333, time=0.63, value=1}
         {arriveTangent=-5.3669, interpMode=RCIM_Cubic, leaveTangent=-5.3669, tangentMode=RCTM_User, time=1, value=0}
        ]}
       YCurve {keys [
         <at 0 2 RCIM_Cubic>
         {arriveTangent=-4, interpMode=RCIM_Cubic, leaveTangent=-4, time=0.1, value=1}
         <at 0.25 1>
         {arriveTangent=-1.3333, interpMode=RCIM_Cubic, leaveTangent=-1.3333, time=0.63, value=1}
         {arriveTangent=-5.3669, interpMode=RCIM_Cubic, leaveTangent=-5.3669, tangentMode=RCTM_User, time=1, value=0}
        ]}
      }}
     ScaleCurve=<1 1>
    ))
   &ScaleSpriteSize(
    Scale Sprite Size Mode=::NewEnumerator0
    Uniform Scale Factor=&FloatFromCurve(
     FloatCurve=DataInterface() {propertyValues {Curve {keys [<at 500 1 RCIM_Cubic>, {interpMode=RCIM_Cubic, leaveTangent=0.0207, time=1400, value=3}]}}}
     CurveIndex=$Engine.Owner.LODDistance
    )
   ) {module=ScaleSpriteSize001}
   &KillParticles(Kill Particles=&SetBoolByFloatComparison(
     A=&MakeFloatFromLinearColor(LinearColor=$Output.SampleTexture.SampledColor, Channel:ENiagaraLinearColor_Channels=::NewEnumerator1)
     B=0.99
     Comparison Type=::NewEnumerator7
    ))
  ]
  rendererClasses [&NiagaraSpriteRendererProperties]
  renderers [{rendererClass=&NiagaraSpriteRendererProperties}]
 }
 Sparkles_Sides {
  emitterSpawnScript [&FN_PreserveSkydiveUserVariables() {module=PreserveUserVariables}]
  emitterUpdateScript [
   &EmitterState()
   &SpawnRate(Spawn Probability=0.9, Use Spawn Probability=true)
  ]
  particleSpawnScript [
   &FN_SetSpawnGroup(Number of Spawn Groups=2)
   &InitializeParticle(
    Color=<rgba 1 0.6042 0.9784 1>
    Lifetime Max=1.2
    Lifetime Min=0.7
    Lifetime Mode=::NewEnumerator1
    Sprite Size Mode=::NewEnumerator1
    Uniform Sprite Size Max=14
    Uniform Sprite Size Min=8
    UsePositionOffset=false
    Write Material Random=true
   )
   &CameraOffset(Camera Offset Amount=6)
   /SetVariables("Particles.OffsetPosition ":Vector3f=&MakeVector(
     X=&Multiply_Float(
      A=&Lerp_Float(A=7.5, B=11, Alpha=$Particles.MaterialRandom)
      B=&MaskFloatBySpawnGroup(Masked Float=-1, Passthrough Float=1)
     )
     Y=&RandomRangeFloat(
      Minimum=-2
      Maximum=&FloatFromCurve(
       FloatCurve=DataInterface() {propertyValues {Curve {keys [
           {
            arriveTangent=-416.5733
            arriveTangentWeight=9.5512
            interpMode=RCIM_Cubic
            leaveTangent=-416.5733
            leaveTangentWeight=9.5512
            tangentMode=RCTM_User
            tangentWeightMode=RCTWM_WeightedBoth
            time=0
            value=-2
           }
           {
            arriveTangent=17.6788
            arriveTangentWeight=5.8435
            interpMode=RCIM_Cubic
            leaveTangent=17.6788
            leaveTangentWeight=5.8435
            tangentMode=RCTM_User
            tangentWeightMode=RCTWM_WeightedBoth
            time=1
            value=-10
           }
          ]}}}
       CurveIndex=$Particles.MaterialRandom
      )
     )
     Z=&FloatFromCurve(FloatCurve=DataInterface() {propertyValues {Curve {keys [<at 0 6>,<at 1 -20>]}}}, CurveIndex=$Particles.MaterialRandom)
    )) {isSetParameters=true, module=SetVariables_#1}
  ]
  particleUpdateScript [
   &ParticleState()
   &SkeletalMeshLocation(
    Bone Sampling Mode=::NewEnumerator4
    MaskSpawnGroup=false
    Sampled Position Offset=Linked() {linkedVariable {name="Particles.OffsetPosition ", type=Vector3f}}
    Skeletal Mesh=$User.SkeletalMesh
   )
   &ScaleSpriteSize(Scale Factor=&Vector2DFromCurve(
     Vector2Curve=DataInterface() {propertyValues {
       XCurve {keys [
         {arriveTangent=25.6549, interpMode=RCIM_Cubic, leaveTangent=25.6549, tangentMode=RCTM_User, time=0, value=0.75}
         {arriveTangent=1, interpMode=RCIM_Cubic, leaveTangent=1, time=0.1, value=0.28}
         {arriveTangent=1.3585, leaveTangent=1.3585, time=0.25, value=1}
         {arriveTangent=-1.3333, interpMode=RCIM_Cubic, leaveTangent=-1.3333, time=0.63, value=1}
         {arriveTangent=-5.3669, interpMode=RCIM_Cubic, leaveTangent=-5.3669, tangentMode=RCTM_User, time=1, value=0}
        ]}
       YCurve {keys [
         <at 0 2 RCIM_Cubic>
         {arriveTangent=-4, interpMode=RCIM_Cubic, leaveTangent=-4, time=0.1, value=1}
         <at 0.25 1>
         {arriveTangent=-1.3333, interpMode=RCIM_Cubic, leaveTangent=-1.3333, time=0.63, value=1}
         {arriveTangent=-5.3669, interpMode=RCIM_Cubic, leaveTangent=-5.3669, tangentMode=RCTM_User, time=1, value=0}
        ]}
      }}
     ScaleCurve=<1 1>
    ))
   &ScaleSpriteSize(
    Scale Sprite Size Mode=::NewEnumerator0
    Uniform Scale Factor=&FloatFromCurve(
     FloatCurve=DataInterface() {propertyValues {Curve {keys [<at 500 1 RCIM_Cubic>, {interpMode=RCIM_Cubic, leaveTangent=0.0207, time=1400, value=3}]}}}
     CurveIndex=$Engine.Owner.LODDistance
    )
   ) {module=ScaleSpriteSize001}
  ]
  rendererClasses [&NiagaraSpriteRendererProperties]
  renderers [{rendererClass=&NiagaraSpriteRendererProperties}]
 }
}
summary {
 systemName=NS_Twin_Dark_Purple_Backpack
 emitters [
  {emitterName=Sparkles_Front, rendererClasses [&NiagaraSpriteRendererProperties]}
  {emitterName=Sparkles_GPU, simTarget=GPUComputeSim, rendererClasses [&NiagaraSpriteRendererProperties]}
  {emitterName=Sparkles_Sides, rendererClasses [&NiagaraSpriteRendererProperties]}
  {emitterName=Sparkles_Bottom, rendererClasses [&NiagaraSpriteRendererProperties]}
 ]
 userVariables [
  User.bIsFrontEndPreview {type=NiagaraBool}
  User.PlayerSpeed {type=NiagaraFloat, defaultValue=0, description=""}
  User.SkeletalMesh {
   type=NiagaraDataInterfaceSkeletalMesh
   defaultValue {
    struct=NiagaraExt_VariableValue_DataInterface
    value {dataInterface=&self:NiagaraDataInterfaceSkeletalMesh_1}
   }
  }
 ]
}
