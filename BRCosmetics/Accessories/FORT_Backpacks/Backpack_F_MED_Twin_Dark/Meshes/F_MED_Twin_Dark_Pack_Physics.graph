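# Physics asset for the F_MED_Twin_Dark backpack cosmetic, setting up physical motion for its dangling zipper.
# A large capsule covers the pack bone, with small capsules on the zipper parent and three zipper bones.
# Constraints chain the zipper bones in series from the pack, limiting swing to narrow angles and setting twist limits to zero.
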PhysicsAsset /BRCosmetics/Accessories/FORT_Backpacks/Backpack_F_MED_Twin_Dark/Meshes/F_MED_Twin_Dark_Pack_Physics
bodies {
 dyn_zipper_01 [{shapeType=Capsule, length=0.8307, radius=0.5, shapeName=dyn_zipper_01_capsule, center=<1.2 0 0>, rotation=<rot 90 0 0>}]
 dyn_zipper_02 [{shapeType=Capsule, length=1.1542, radius=0.5, shapeName=dyn_zipper_02_capsule, center=<1.2 0 0>, rotation=<rot 90 0 0>}]
 dyn_zipper_03 [{shapeType=Capsule, length=0.4206, radius=0.6747, shapeName=dyn_zipper_03_capsule, center=<1 0 0>, rotation=<rot 90 0 0>}]
 dyn_zipper_parent [{shapeType=Capsule, length=0.4506, radius=0.5, shapeName=dyn_zipper_parent_capsule, center=<0.75 0 0>, rotation=<rot 90 0 0>}]
 pack [{shapeType=Capsule, length=5.4712, radius=9.11, shapeName=pack_capsule, center=<0 -2 -0.5>, rotation=<rot 90 -90 -90.0002>}]
}
constraints [
 {bone1Name=dyn_zipper_01, bone2Name=dyn_zipper_parent, swing1LimitDegrees=15, swing2LimitDegrees=25, twistLimitDegrees=0}
 {bone1Name=dyn_zipper_02, bone2Name=dyn_zipper_01, swing1LimitDegrees=15, swing2LimitDegrees=25, twistLimitDegrees=0}
 {bone1Name=dyn_zipper_03, bone2Name=dyn_zipper_02, swing1LimitDegrees=15, swing2LimitDegrees=25, twistLimitDegrees=0}
 {bone1Name=dyn_zipper_parent, bone2Name=pack, swing1LimitDegrees=10, swing2LimitDegrees=25, twistLimitDegrees=0}
]
